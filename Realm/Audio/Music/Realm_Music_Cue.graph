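# Sound cue for Realm music that shifts between looping 124 BPM idle and aggro tracks, routed to the NPC music sound class.
# A mixer sums three continuous modulators whose volumes follow the Unaware, Alerted and Aggressive parameters, each defaulting to zero.
# Unaware scales the idle loop, while Alerted and Aggressive share one aggro wave player, so both states use the same track.
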SoundCue /Realm/Audio/Music/Realm_Music_Cue
root=Mixer_1
nodes [
 {id=Mixer_1, children [ModulatorContinuous_0, ModulatorContinuous_1, ModulatorContinuous_2], properties {inputVolume [1,1,1]}}
 {id=ModulatorContinuous_0, children [WavePlayer_0], properties {volumeModulationParams {default=0, maxOutput=2, parameterName=Unaware}}}
 {id=ModulatorContinuous_1, children [WavePlayer_3], properties {volumeModulationParams {default=0, maxOutput=2, parameterName=Alerted}}}
 {id=ModulatorContinuous_2, children [WavePlayer_3], properties {volumeModulationParams {default=0, maxOutput=2, parameterName=Aggressive}}}
 {id=WavePlayer_0, properties {bLooping=true, soundWaveAssetPtr=&/Realm_Idle_124BPM}}
 {id=WavePlayer_3, properties {bLooping=true, soundWaveAssetPtr=&/Realm_Aggro_124BPM}}
]
properties {maxDistance=2097152, soundClassObject=&Fort_Music_NPC, volumeMultiplier=0.531}
