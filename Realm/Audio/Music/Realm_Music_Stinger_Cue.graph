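# Sound cue for a Realm music stinger, built on the Realm_End_124BMP wave.
# A single wave player feeds the cue output directly, with no randomization, mixing or modulation nodes.
# The cue routes through the Fort_Music_NPC sound class, boosts volume above default and sets a very large max distance.
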SoundCue /Realm/Audio/Music/Realm_Music_Stinger_Cue
root=WavePlayer_0
nodes [{id=WavePlayer_0, properties {soundWaveAssetPtr=&/Realm_End_124BMP}}]
properties {maxDistance=2097152, soundClassObject=&Fort_Music_NPC, volumeMultiplier=1.4}
